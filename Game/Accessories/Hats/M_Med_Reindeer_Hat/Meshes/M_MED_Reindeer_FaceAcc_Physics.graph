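# Physics asset for the medium reindeer hat face accessory, defining bodies and joints for the movement of its ears and holly.
# Each ear is a two-capsule chain hinged to the head, with looser limits on the outer segment than at the base.
# Two holly capsules attach directly to the head with tight limits that favor one swing axis.
# The head body combines a sphere with an angled box.
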PhysicsAsset /Game/Accessories/Hats/M_Med_Reindeer_Hat/Meshes/M_MED_Reindeer_FaceAcc_Physics
bodies {
 dyn_ear_01_l [{shapeType=Capsule, length=1.3198, radius=1.2195, center=<-0.2774 0 0.0001>, rotation=<rot 90 0 0>}]
 dyn_ear_01_r [{shapeType=Capsule, length=1.4448, radius=0.9695, center=<0.6764 0 -0.0001>, rotation=<rot 90 0 0>}]
 dyn_ear_02_l [{shapeType=Capsule, length=1.842, radius=1.0637, center=<-0.4082 0 0.0001>, rotation=<rot 90 0 0>}]
 dyn_ear_02_r [{shapeType=Capsule, length=2.092, radius=0.9387, center=<0.4145 0 0.0001>, rotation=<rot 90 0 0>}]
 dyn_holly1 [{shapeType=Capsule, length=2.568, radius=0.9574, center=<2.2614 0.2509 -0.0512>, rotation=<rot 0 -86.5944 -86.0912>}]
 dyn_holly2 [{shapeType=Capsule, length=3.2198, radius=0.7893, center=<2.6195 0.2704 -0.064>, rotation=<rot 0 -86.9779 -86.5598>}]
 head [
  {shapeType=Sphere, radius=7.4129, center=<8.7654 0.0003 0>}
  {shapeType=Box, extentX=7.5, extentY=2, extentZ=7.5, center=<14.1199 1.5257 -6.6562>, rotation=<rot -39.0817 -22.137 5.932>}
 ]
}
constraints [
 {bone1Name=dyn_ear_01_l, bone2Name=head, swing1LimitDegrees=15, swing2LimitDegrees=15, twistLimitDegrees=5}
 {bone1Name=dyn_ear_02_l, bone2Name=dyn_ear_01_l, swing1LimitDegrees=20, swing2LimitDegrees=20, twistLimitDegrees=10}
 {bone1Name=dyn_ear_01_r, bone2Name=head, swing1LimitDegrees=15, swing2LimitDegrees=15, twistLimitDegrees=5}
 {bone1Name=dyn_ear_02_r, bone2Name=dyn_ear_01_r, swing1LimitDegrees=20, swing2LimitDegrees=20, twistLimitDegrees=10}
 {bone1Name=dyn_holly1, bone2Name=head, swing1LimitDegrees=10, swing2LimitDegrees=2, twistLimitDegrees=2}
 {bone1Name=dyn_holly2, bone2Name=head, swing1LimitDegrees=10, swing2LimitDegrees=2, twistLimitDegrees=2}
]
