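# Physics asset for the tail of the medium reindeer hat accessory, defining its collision bodies and joints for physical motion.
# Capsules cover the two dyn_tail bones, and an offset sphere sits on the pelvis at the root of the chain.
# Two constraints link pelvis to dyn_tail_01 to dyn_tail_02, each allowing a moderate, symmetric swing cone with very little twist, so the tail can sway but barely roll.
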PhysicsAsset /Game/Accessories/Hats/M_Med_Reindeer_Hat/Meshes/M_MED_Reinder_tail_Physics
bodies {
 dyn_tail_01 [{shapeType=Capsule, length=5.4047, radius=1.7574, center=<5.3752 0.4813 0.1899>, rotation=<rot 90 -0.0001 -0.0001>}]
 dyn_tail_02 [{shapeType=Capsule, length=5.6018, radius=2.5005, center=<5.7411 0 0>, rotation=<rot 0 90 90>}]
 pelvis [{shapeType=Sphere, radius=4.6541, center=<3.0123 -6.3765 0.2791>}]
}
constraints [
 {bone1Name=dyn_tail_01, bone2Name=pelvis, swing1LimitDegrees=25, swing2LimitDegrees=25, twistLimitDegrees=2}
 {bone1Name=dyn_tail_02, bone2Name=dyn_tail_01, swing1LimitDegrees=25, swing2LimitDegrees=25, twistLimitDegrees=2}
]
