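# Animation Blueprint for the medium male reindeer hat accessory, adding secondary physics and adapting to a worn cape.
# The AnimGraph copies the pose from the attached parent mesh, then chains two RigidBody nodes using face-accessory and tail physics assets in base-bone space.
# The update event flags a cape once if the backpack's cape_length socket is far enough from backpack_root.
# When flagged, it sets the "Hide Element 01" parameter on a dynamic material instance.
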AnimBlueprint /BRCosmetics/Accessories/Hats/M_Med_Reindeer_Hat/Meshes/M_MED_Reindeer_FaceAcc_Hat_MESH_Skeleton_AnimBlueprint
target_skeleton=/Game/Accessories/Hats/M_Med_Reindeer_Hat/Meshes/M_MED_Reindeer_FaceAcc_Hat_MESH_Skeleton
graphs [
 AnimGraph {nodes [
   n1 = RigidBody(ComponentPose=Local To Component(LocalPose=Copy Pose From Mesh() {bUseAttachedParent=true})) {
    bOverrideWorldGravity=true
    overridePhysicsAsset=/Game/Accessories/Hats/M_Med_Reindeer_Hat/Meshes/M_MED_Reindeer_FaceAcc_Physics
    simulationSpace=BaseBoneSpace
    alphaBoolBlend {blendOption=Linear}
    baseBoneRef {boneName=spine_04}
    componentAppliedLinearAccClamp=<1000 1000 1000>
    componentLinearAccScale=<0.2 0.2 0.2>
    componentLinearVelScale=<0.5 0.5 0.5>
    overrideWorldGravity {z=-980}
   }
   n2 = RigidBody(ComponentPose=@n1) {
    bOverrideWorldGravity=true
    overridePhysicsAsset=/Game/Accessories/Hats/M_Med_Reindeer_Hat/Meshes/M_MED_Reinder_tail_Physics
    simulationSpace=BaseBoneSpace
    alphaBoolBlend {blendOption=Linear}
    componentAppliedLinearAccClamp=<1000 1000 1000>
    componentLinearAccScale=<0.5 0.5 0.5>
    componentLinearVelScale=<0.3 0.3 0.3>
    overrideWorldGravity {z=-980}
   }
   Output Pose(Result=Component To Local(ComponentPose=@n2)) {layerGroup=None}
  ]}
 EventGraph {nodes [
   Event BlueprintUpdateAnimation()
   > n1 = Cast To FortPlayerPawn(Object=GetOwningActor())
   > n2 = Sequence!(execute) -> (then_0, then_1, then_2:Exec, then_3:Exec)
   n3 = GetSkeletalMeshForPartType(self=@n1, PartType=Backpack)
   > Is Valid(InputObject=@n3)
   > Is Valid(InputObject=FindSocket!(
     self:Skinned Asset Object Reference=Get SkeletalMesh!(self:Skinned Mesh Component Object Reference=@n3) -> (SkeletalMesh:Skeletal Mesh Object Reference)
     InSocketName:Name=cape_length
    ) -> (ReturnValue:Skeletal Mesh Socket Object Reference))
   > Do Once()
   > Set isWearingCape(isWearingCape="float > float"(
     A=Get Distance Between Two Sockets!(
      Component:Skeletal Mesh Component Object Reference=@n3
      SocketOrBoneNameA:Name=backpack_root
      SocketSpaceA:ERelativeTransformSpace Enum=RTS_World
      SocketOrBoneNameB:Name=cape_length
      SocketSpaceB:ERelativeTransformSpace Enum=RTS_World
      bRemapRange:bool=false
      InRangeMin:float=0
      InRangeMax:float=0
      OutRangeMin:float=0
      OutRangeMax:float=0
     ) -> (ReturnValue:float) {is=title K2Node_CallFunction}
     B=50
    ) {is=title K2Node_CallFunction})
   Branch(execute=@n2.then_1, Condition=Get isWearingCape() -> (isWearingCape:bool))
   > n4 = Get Dynamic Material Instance(SkeletalMeshComponent=GetPartSkeletalMeshComponent())
   > SetScalarParameterValue(self=@n4, ParameterName=Hide Element 01, Value=1)
  ]}
]
variables [isWearingCape {type=boolean}]
